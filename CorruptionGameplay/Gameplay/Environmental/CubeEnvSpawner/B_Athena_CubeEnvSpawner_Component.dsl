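; Server-side component that scatters shadow stone spawners and corrupted jump vents around a Corruption cube environment spawner.
; Once the playlist is ready and a short random delay staggers instances, it runs an EQS query on the owner and shuffles the resulting locations.
; It spawns stone spawners first, then vents, with random counts between scalable-float bounds, Z offsets and random yaw.
; Both draw from one shared location pool, removing each used point, so counts may fall short.
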(event EventBeginPlay
  ;[1 1 1] Begin Play - Wait for playlist then add some varience to when to do the spawning. After start the EQS.
  (if (and (Networking|IsServer) (Variables|Default|GetEnabled))
    (Class|FortAsyncActionCurrentPlaylistReady|PlaylistDataReadyAsync
      (:OnReady
        (Utilities|FlowControl|Delay (Math|Random|RandomFloatinRange 0.2 2.0)) ; Make sure they happen at different times.
        (CallFunction|StartEQS)))))

(event Custom|StartEQS
  ;[1 1 1] Run the EQS and begin spawning when it has finished.
  (bind _returnvalue (AI|EQS|RunEQSQuery (Variables|Default|GetEQS) (Components|GetOwner) "AllMatching"))
  (AI|EQS|BindEventtoOnQueryFinished _returnvalue (delegate Custom|EQSDone)))

(event Custom|EQSDone
  (params (QueryInstance :type EQS%20Query%20Instance%20Object%20Reference) (QueryStatus :type EEnvQueryStatus%20Enum :default "Processing"))
  ;[1 1 1] Run the EQS and begin spawning when it has finished.
  (bind (_resultlocations _returnvalue) (AI|EQS|GetQueryResultsasLocations QueryInstance))
  (if _returnvalue
    (bind _output_get (Variables|Default|SetEnvironmentalActorLocations _resultlocations))
    (Utilities|Array|Shuffle _output_get)
    (CallFunction|SpawnShadowStones)))

(event Custom|SpawnVents
  (bind _environmentalactorlocations (Variables|Default|GetEnvironmentalActorLocations))
  (bind (_outitem _outindex) (Utilities|Array|RandomArrayItem _environmentalactorlocations))
  ;[1 1 1] Spawn Vents - Spawn in the vents based on how many were selected to attempt to spawn. Note the min may be less than the desired amount if there are not good available locations.
  (bind _output_get (Variables|Default|SetTempSpawnInt (CallFunction|GetRandomSpawnCount (Variables|Default|GetRow_MinVents) (Variables|Default|GetRow_MaxVents))))
  (if (> _output_get 0)
    (for _ (range 1 (+ _output_get 1))
      (bind _output_get_2 (Variables|Default|SetTempIndexForSpawn _outindex))
      (if (> (Utilities|Array|Length _environmentalactorlocations) 0)
        (Game|SpawnActorfromClass "/CorruptionGameplay/Gameplay/Environmental/JumpVents/BP_Corrupted_Vent.BP_Corrupted_Vent_C" (Math|Transform|MakeTransform (Math|Vector|vector+vector _outitem (Math|Vector|MakeVector 0.0 0.0 (Variables|Default|GetVentZSpawnOffset))) (Math|Rotator|MakeRotator 0.0 0.0 (Math|Random|RandomFloatinRange 0.0 359.0)) "1.000000,1.000000,1.000000") "Undefined" "MultiplyWithRoot" 0 1.0 1.5 0 0 0)
        (Utilities|Array|RemoveIndex _environmentalactorlocations _output_get_2)))))

(event Custom|SpawnShadowStones
  (bind _environmentalactorlocations (Variables|Default|GetEnvironmentalActorLocations))
  (bind (_outitem _outindex) (Utilities|Array|RandomArrayItem _environmentalactorlocations))
  ;[1 1 1] Spawn Shadow Stone - Spawn in the stone spawners  based on how many were selected to attempt to spawn. Note the min may be less than the desired amount if there are not good available locations.
  (bind _output_get (Variables|Default|SetTempSpawnInt (CallFunction|GetRandomSpawnCount (Variables|Default|GetRow_MinShadowStoneSpawners) (Variables|Default|GetRow_MaxShadowStoneSpawners))))
  (if (> _output_get 0)
    (for _ (range 1 (+ _output_get 1))
      (bind _output_get_2 (Variables|Default|SetTempIndexForSpawn _outindex))
      (if (> (Utilities|Array|Length _environmentalactorlocations) 0)
        (Game|SpawnActorfromClass (Variables|Default|GetStoneSpawnerClass) (Math|Transform|MakeTransform (Math|Vector|vector+vector _outitem (Math|Vector|MakeVector 0.0 0.0 (Variables|Default|GetStoneSpawnerZSpawnOffset))) (Math|Rotator|MakeRotator 0.0 0.0 (Math|Random|RandomFloatinRange 0.0 359.0)) "1.000000,1.000000,1.000000") "Undefined" "SelectDefaultAtRuntime" 0 0)
        (Utilities|Array|RemoveIndex _environmentalactorlocations _output_get_2)))
    (CallFunction|SpawnVents)
    (else
      (CallFunction|SpawnVents))))

(fn GetRandomSpawnCount
  (params (Min :type Scalable%20Float%20Structure%20%28by%20ref%29) (Max :type Scalable%20Float%20Structure%20%28by%20ref%29))
  ;[1 1 1] Gets the value of the row then picks a random from the min and max
  (return (Math|Random|RandomIntegerinRange (Math|Float|Floor (ScalableFloat|GetValueatLevel Min)) (Math|Float|Floor (ScalableFloat|GetValueatLevel Max)))))
