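# EQS query that picks placement points for Corruption cube environmentals, judging by its name and CubeEnvSpawner folder.
# A donut generator creates rings of navmesh-projected candidates around the querier.
# Points inside Athena water volumes are filtered out, and a nearby-buildings test both filters and scores.
# Flat-surface and weighted random tests also apply, while linear 2D distance scoring favors points farther from the querier.
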EnvQuery /CorruptionGameplay/Gameplay/Environmental/CubeEnvSpawner/EQS_Donut_Cube_Environmentals
options [{
  generator {
   name=EnvQueryGenerator_Donut_0
   class=EnvQueryGenerator_Donut
   properties {
    bUseSpiralPattern=false
    center=&EnvQueryContext_Querier
    innerRadius=2500
    numberOfRings=3
    outerRadius=5000
    pointsPerRing=12
    projectionData {extentX=0, navigationFilter=&RecastFilter_UseDefaultArea, postProjectionVerticalOffset=0, projectDown=2048}
   }
  }
  tests [
   FortQueryTest_Random_1 {class=FortQueryTest_Random, properties {floatValueMax=0, scoreClampMax=1, scoringFactor=3}}
   EnvQueryTest_Volume_0 {
    class=EnvQueryTest_Volume
    properties {boolValue=false, testPurpose=Filter, volumeClass=/Script/FortniteGame.WaterVolume_Athena, volumeContext=null}
   }
   EnvQueryTest_Distance_1 {
    class=EnvQueryTest_Distance
    properties {
     distanceTo=&EnvQueryContext_Querier
     filterType=Range
     floatValueMax=5000
     floatValueMin=1
     scoringEquation=Linear
     testComment=""
     testMode=Distance2D
     testPurpose=Score
     scoringFactor {dataBinding=null, dataField=FloatValue, defaultValue=5}
    }
   }
   FortQueryTest_OnFlatSurface_2 {class=FortQueryTest_OnFlatSurface, properties {radius=200, scoringFactor=5, toleranceZ=50, traceOffsetUp=100}}
   FortQueryTest_HasNearbyBuildings_0 {class=FortQueryTest_HasNearbyBuildings, properties {extentXY=2, testPurpose=FilterAndScore}}
  ]
 }]
